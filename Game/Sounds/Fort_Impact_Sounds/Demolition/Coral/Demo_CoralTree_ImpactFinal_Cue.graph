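# Sound cue for the final demolition impact on a coral tree, built from borrowed tree, mushroom and rock impact sounds.
# A mixer layers three random picks: big tree hits, full-level mushroom impacts and quieter rock impacts.
# An enveloper ramps the tree layer up from low volume, and a top modulator varies the mix.
# Small-object destruction attenuation and an impacts sound class apply, with object-impact concurrency capped at four.
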SoundCue /Game/Sounds/Fort_Impact_Sounds/Demolition/Coral/Demo_CoralTree_ImpactFinal_Cue
root=Modulator_2
nodes [
 {
  id=Enveloper_1
  children [Random_3]
  properties {
   pitchCurve {editorCurveData {keys [{time=0}]}}
   volumeCurve {editorCurveData {keys [{interpMode=RCIM_Cubic, tangentMode=RCTM_Break, time=0, value=0.1}, <at 0.5 1>]}}
  }
 }
 {id=Mixer_1, children [Enveloper_1,Random_2,Random_4], properties {inputVolume [0.7,1,0.5]}}
 {id=Modulator_2, children [Mixer_1]}
 {id=Random_2, children [WavePlayer_3, WavePlayer_10], properties {weights [1,1]}}
 {id=Random_3, children [WavePlayer_7,WavePlayer_8,WavePlayer_9], properties {bShouldExcludeFromBranchCulling=true, weights [1,1,1]}}
 {id=Random_4, children [WavePlayer_11, WavePlayer_12], properties {bShouldExcludeFromBranchCulling=true, weights [1,1]}}
 {id=WavePlayer_10, properties {soundWaveAssetPtr=/Game/Sounds/Fort_Impact_Sounds/Demolition/Mushroom/Demo_Mushroom_Impact_01}}
 {id=WavePlayer_11, properties {soundWaveAssetPtr=/Game/Sounds/Fort_Impact_Sounds/Demolition/Defaults/Fort_Impact_RockFinal_01}}
 {id=WavePlayer_12, properties {soundWaveAssetPtr=/Game/Sounds/Fort_Impact_Sounds/Demolition/Defaults/Fort_Impact_RockFinal_02}}
 {id=WavePlayer_3, properties {soundWaveAssetPtr=/Game/Sounds/Fort_Impact_Sounds/Demolition/Mushroom/Demo_Mushroom_Impact_02}}
 {id=WavePlayer_7, properties {soundWaveAssetPtr=/Game/Sounds/Fort_Impact_Sounds/Demolition/Tree/tree_big_hit_01}}
 {id=WavePlayer_8, properties {soundWaveAssetPtr=/Game/Sounds/Fort_Impact_Sounds/Demolition/Tree/tree_big_hit_03}}
 {id=WavePlayer_9, properties {soundWaveAssetPtr=/Game/Sounds/Fort_Impact_Sounds/Demolition/Tree/tree_big_hit_07}}
]
properties {
 attenuationSettings=/Game/Sounds/Attenuation/Fort_Demolition_Destroy_Small_Sized_Object_Att
 maxDistance=4000
 soundClassObject=&Fort_Impacts_PSM
 volumeMultiplier=0.68
 concurrencyOverrides {maxCount=4}
 concurrencySet [&CONC_Object_Impacts_FourMax]
}
