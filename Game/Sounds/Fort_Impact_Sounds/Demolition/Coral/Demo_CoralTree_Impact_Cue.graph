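# Sound cue for coral tree demolition impacts, layered from borrowed tree, mushroom and rock hit sounds.
# A mixer blends random tree hits with a volume swell-in envelope, random mushroom impacts, and a modulated rock impact.
# The mix passes through a fade-out envelope and a slight downward pitch randomizer for variation.
# It plays in the impacts sound class with small-object hit attenuation and a four-voice concurrency limit.
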SoundCue /Game/Sounds/Fort_Impact_Sounds/Demolition/Coral/Demo_CoralTree_Impact_Cue
root=Modulator_2
nodes [
 {
  id=Enveloper_0
  children [Mixer_0]
  properties {pitchCurve {editorCurveData {keys [{time=0}]}}, volumeCurve {editorCurveData {keys [<at 0 1 RCIM_Cubic>,<at 1.5429 -0.0334>]}}}
 }
 {
  id=Enveloper_1
  children [Random_1]
  properties {
   pitchCurve {editorCurveData {keys [{time=0}]}}
   volumeCurve {editorCurveData {keys [{interpMode=RCIM_Cubic, tangentMode=RCTM_Break, time=0, value=0.1}, <at 0.5 1>]}}
  }
 }
 {id=Mixer_0, children [Enveloper_1,Random_12,Modulator_0], properties {inputVolume [0.7,0.55,1]}}
 {id=Modulator_0, children [WavePlayer_6]}
 {id=Modulator_2, children [Enveloper_0], properties {pitchMax=1, pitchMin=0.92}}
 {id=Random_1, children [WavePlayer_3,WavePlayer_4,WavePlayer_7], properties {weights [1,1,1]}}
 {id=Random_12, children [WavePlayer_0,WavePlayer_1], properties {weights [1,1]}}
 {id=WavePlayer_0, properties {soundWaveAssetPtr=/Game/Sounds/Fort_Impact_Sounds/Demolition/Mushroom/Demo_Mushroom_Impact_02}}
 {id=WavePlayer_1, properties {soundWaveAssetPtr=/Game/Sounds/Fort_Impact_Sounds/Demolition/Mushroom/Demo_Mushroom_Impact_01}}
 {id=WavePlayer_3, properties {soundWaveAssetPtr=/Game/Sounds/Fort_Impact_Sounds/Demolition/Tree/tree_big_hit_01}}
 {id=WavePlayer_4, properties {soundWaveAssetPtr=/Game/Sounds/Fort_Impact_Sounds/Demolition/Tree/tree_big_hit_03}}
 {id=WavePlayer_6, properties {soundWaveAssetPtr=/Game/Sounds/Fort_Impact_Sounds/Demolition/Defaults/rock_newimp_03}}
 {id=WavePlayer_7, properties {soundWaveAssetPtr=/Game/Sounds/Fort_Impact_Sounds/Demolition/Tree/tree_big_hit_07}}
]
properties {
 attenuationSettings=/Game/Sounds/Attenuation/Fort_Demolition_Hit_Small_Sized_Object_Att
 maxDistance=4000
 soundClassObject=&Fort_Impacts
 volumeMultiplier=0.68
 concurrencyOverrides {maxCount=4}
 concurrencySet [&CONC_Object_Impacts_FourMax]
}
