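# Sound cue for the final impact when coral is demolished, built from borrowed mushroom, flower and rock sounds.
# Three layers each pick one of two waves at random and are mixed, with the flower layer briefly delayed and quieter.
# An enveloper fades the mix out over about 1.5 seconds, then a modulator slightly randomizes pitch downward.
# Small-object demolition attenuation and a four-voice concurrency limit govern playback.
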SoundCue /Game/Sounds/Fort_Impact_Sounds/Demolition/Coral/Demo_Coral_ImpactFinal_Cue
root=Modulator_2
nodes [
 {id=Delay_0, children [Random_0], properties {delayMax=0.1, delayMin=0.1}}
 {
  id=Enveloper_0
  children [Mixer_0]
  properties {pitchCurve {editorCurveData {keys [{time=0}]}}, volumeCurve {editorCurveData {keys [<at 0 1 RCIM_Cubic>,<at 1.5429 -0.0334>]}}}
 }
 {id=Mixer_0, children [Random_12,Delay_0,Random_1], properties {inputVolume [1,0.65,1]}}
 {id=Modulator_2, children [Enveloper_0], properties {pitchMax=1, pitchMin=0.92}}
 {id=Random_0, children [WavePlayer_2,WavePlayer_5], properties {weights [1,1]}}
 {id=Random_1, children [WavePlayer_4,WavePlayer_6], properties {weights [1,1]}}
 {id=Random_12, children [WavePlayer_0,WavePlayer_1], properties {weights [1,1]}}
 {id=WavePlayer_0, properties {soundWaveAssetPtr=/Game/Sounds/Fort_Impact_Sounds/Demolition/Mushroom/Demo_Mushroom_Impact_02}}
 {id=WavePlayer_1, properties {soundWaveAssetPtr=/Game/Sounds/Fort_Impact_Sounds/Demolition/Mushroom/Demo_Mushroom_Impact_01}}
 {id=WavePlayer_2, properties {soundWaveAssetPtr=/Game/Sounds/Fort_Impact_Sounds/Demolition/Flowers/Demo_Flowers_01}}
 {id=WavePlayer_4, properties {soundWaveAssetPtr=/Game/Sounds/Fort_Impact_Sounds/Demolition/Defaults/rock_newimp_10}}
 {id=WavePlayer_5, properties {soundWaveAssetPtr=/Game/Sounds/Fort_Impact_Sounds/Demolition/Flowers/Demo_Flowers_04}}
 {id=WavePlayer_6, properties {soundWaveAssetPtr=/Game/Sounds/Fort_Impact_Sounds/Demolition/Defaults/rock_newimp_11}}
]
properties {
 attenuationSettings=/Game/Sounds/Attenuation/Fort_Demolition_Destroy_Small_Sized_Object_Att
 maxDistance=4000
 soundClassObject=&Fort_Impacts_PSM
 volumeMultiplier=0.68
 concurrencyOverrides {maxCount=4}
 concurrencySet [&CONC_Object_Impacts_FourMax]
}
